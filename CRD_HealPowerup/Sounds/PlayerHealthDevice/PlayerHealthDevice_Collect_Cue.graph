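# Sound cue played when a player collects the heal powerup from the health device.
# A mixer layers a random pick of three equally weighted positive device sounds with a quieter, pitched-up butterfly spawn sound, each layer passing through its own modulator.
# It routes to the critical in-game UI sound class and overrides attenuation for short-range falloff and concurrency to cap simultaneous instances with volume scaling.
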SoundCue /CRD_HealPowerup/Sounds/PlayerHealthDevice/PlayerHealthDevice_Collect_Cue
root=Mixer_1
nodes [
 {id=Mixer_1, children [Modulator_1,Modulator_0], properties {inputVolume [1,0.75]}}
 {id=Modulator_0, children [WavePlayer_4], properties {pitchMax=1.2, pitchMin=1.1}}
 {id=Modulator_1, children [Random_0]}
 {id=Random_0, children [WavePlayer_0,WavePlayer_1,WavePlayer_2], properties {weights [1,1,1]}}
 {id=WavePlayer_0, properties {soundWaveAssetPtr=/Game/Sounds/Creative/Toys/PlayerHealthDevice/PlayerHealthDevice_Positive_01}}
 {id=WavePlayer_1, properties {soundWaveAssetPtr=/Game/Sounds/Creative/Toys/PlayerHealthDevice/PlayerHealthDevice_Positive_02}}
 {id=WavePlayer_2, properties {soundWaveAssetPtr=/Game/Sounds/Creative/Toys/PlayerHealthDevice/PlayerHealthDevice_Positive_03}}
 {id=WavePlayer_4, properties {soundWaveAssetPtr=/Game/Athena/Sounds/Quests/S16_Narrative/Butterfly_Spawn_In_01}}
]
properties {
 bOverrideAttenuation=true
 bOverrideConcurrency=true
 maxDistance=727
 soundClassObject=&Fort_UI_InGame_Critical
 attenuationOverrides {falloffDistance=512, stereoSpread=5000, attenuationShapeExtents {x=215}}
 concurrencyOverrides {maxCount=12, volumeScale=0.7}
}
