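# Looping ambient sound cue for the player health device in the heal powerup plugin.
# A looping wave player feeds a modulator that randomly raises pitch slightly, under a reduced volume and raised pitch multiplier on the cue.
# Its own attenuation override uses logarithmic falloff with listener focus, which quiets the loop when it is out of focus.
# It uses the shared loot ambient loop sound class and loot ambient concurrency set.
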SoundCue /CRD_HealPowerup/Sounds/PlayerHealthDevice/PlayerHealthDevice_Loop_Cue
root=Modulator_2
nodes [
 {id=Modulator_2, children [WavePlayer_0], properties {pitchMax=1.15, pitchMin=1, volumeMax=1, volumeMin=1}}
 {id=WavePlayer_0, properties {bLooping=true, soundWaveAssetPtr=&/PlayerHealthDevice_Amb_Loop_01}}
]
properties {
 bOverrideAttenuation=true
 maxDistance=800
 pitchMultiplier=1.13
 soundClassObject=/Game/Sounds/FortSoundClassesAndModes/Fort_Loot_Ambient_Loop
 volumeMultiplier=0.3
 attenuationOverrides {
  bEnableFocusInterpolation=true
  bEnableListenerFocus=true
  distanceAlgorithm=Logarithmic
  falloffDistance=700
  focusAttackInterpSpeed=20
  focusAzimuth=20
  focusReleaseInterpSpeed=10
  lPFFrequencyAtMax=10000
  lPFRadiusMax=1000
  lPFRadiusMin=500
  nonFocusAzimuth=40
  nonFocusVolumeAttenuation=0.6
  occlusionInterpolationTime=0.3
  occlusionLowPassFilterFrequency=3000
  occlusionVolumeAttenuation=0.75
  stereoSpread=500
  attenuationShapeExtents {x=100}
  customAttenuationCurve {editorCurveData {keys [<at 0 1 RCIM_Cubic>, {arriveTangent=-1, interpMode=RCIM_Cubic, leaveTangent=-1, time=0.5, value=0.1}, <at 1 0 RCIM_Cubic>]}}
 }
 concurrencyOverrides {bVolumeScaleCanRelease=true, maxCount=3, volumeScaleAttackTime=0.1, volumeScaleReleaseTime=0.15}
 concurrencySet [/Game/Sounds/Concurrency_Rules/Loot_Ambient_Conc]
}
