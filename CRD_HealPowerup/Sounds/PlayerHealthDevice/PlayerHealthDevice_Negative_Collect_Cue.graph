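# Sound cue for the negative collect sound of the heal powerup's player health device.
# A mixer layers pitched-up random butterfly collect sounds, louder random body hit markers, and a quieter, heavily pitched-down butterfly spawn-in.
# An enveloper on the spawn-in layer fades its volume out while its pitch curve rises.
# The cue plays in the critical in-game UI sound class with its own short-range attenuation and concurrency overrides.
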SoundCue /CRD_HealPowerup/Sounds/PlayerHealthDevice/PlayerHealthDevice_Negative_Collect_Cue
root=Mixer_1
nodes [
 {
  id=Enveloper_1
  children [WavePlayer_16]
  properties {
   pitchCurve {editorCurveData {keys [{time=0}, <at 0.9784 0.9484>]}}
   volumeCurve {editorCurveData {keys [<at 0 1>,<at 0.2929 0.9948>,<at 0.9171 0.2982>,<at 2.1715 -0.0133>]}}
  }
 }
 {id=Mixer_1, children [Modulator_1,Modulator_0,Modulator_2], properties {inputVolume [1,1.4,0.3]}}
 {id=Modulator_0, children [Random_2]}
 {id=Modulator_1, children [Random_1], properties {pitchMax=1.9, pitchMin=1.8}}
 {id=Modulator_2, children [Enveloper_1], properties {pitchMax=0.45, pitchMin=0.4}}
 {id=Random_1, children [WavePlayer_8,WavePlayer_9], properties {bShouldExcludeFromBranchCulling=true, weights [1,1]}}
 {id=Random_2, children [WavePlayer_10, WavePlayer_11, WavePlayer_12], properties {bShouldExcludeFromBranchCulling=true, weights [1,1,1]}}
 {id=WavePlayer_10, properties {soundWaveAssetPtr=/Game/Athena/Sounds/HitNotifications/HitMarker_Body_01}}
 {id=WavePlayer_11, properties {soundWaveAssetPtr=/Game/Athena/Sounds/HitNotifications/HitMarker_Body_02}}
 {id=WavePlayer_12, properties {soundWaveAssetPtr=/Game/Athena/Sounds/HitNotifications/HitMarker_Body_03}}
 {id=WavePlayer_16, properties {soundWaveAssetPtr=/Game/Athena/Sounds/Quests/S16_Narrative/Butterfly_Spawn_In_01}}
 {id=WavePlayer_8, properties {soundWaveAssetPtr=/Game/Athena/Sounds/Quests/S16_Narrative/Butterfly_Collect_01}}
 {id=WavePlayer_9, properties {soundWaveAssetPtr=/Game/Athena/Sounds/Quests/S16_Narrative/Butterfly_Collect_02}}
]
properties {
 bOverrideAttenuation=true
 bOverrideConcurrency=true
 maxDistance=727
 soundClassObject=&Fort_UI_InGame_Critical
 volumeMultiplier=1.1
 attenuationOverrides {falloffDistance=512, stereoSpread=5000, attenuationShapeExtents {x=215}}
 concurrencyOverrides {maxCount=12, volumeScale=0.7}
}
